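# Looping ambient Sound Cue for a smaller government wall tablet computer, played quietly and pitched up at short range.
# A mixer layers an idle buzz loop and a wall tablet loop with computer one-shots that repeat after random delays, with random selection and pitch.
# The quality node routes this mix to its first two slots and leaves the other two empty.
# A concurrency override allows one instance at a time.
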SoundCue /Game/Sounds/Artemis/Environment_Actors/Wall_Tablet/Government_WallTablet_Computer_Smaller_Loop_Cue
root=QualityLevel_0
nodes [
 {id=Delay_2, children [Modulator_4], properties {delayMax=1, delayMin=0.5}}
 {id=Looping_2, children [Delay_2]}
 {id=Mixer_0, children [WavePlayer_4,WavePlayer_2,Looping_2], properties {inputVolume [1,0.5,1]}}
 {id=Modulator_4, children [Random_5], properties {pitchMax=1.3, pitchMin=0.7, volumeMax=1, volumeMin=1}}
 {id=QualityLevel_0, children [Mixer_0,Mixer_0,null,null]}
 {id=Random_5, children [WavePlayer_1, WavePlayer_0, WavePlayer_3, WavePlayer_14], properties {weights [1,1,1,1]}}
 {id=WavePlayer_0, properties {soundWaveAssetPtr=/Game/Sounds/Apollo/Biomes/GovernmentSecretComplex/SecretComplex_Computer_Oneshot_11}}
 {id=WavePlayer_1, properties {soundWaveAssetPtr=/Game/Sounds/Apollo/Biomes/GovernmentSecretComplex/SecretComplex_Computer_Oneshot_10}}
 {id=WavePlayer_14, properties {soundWaveAssetPtr=/Game/Sounds/Apollo/Biomes/GovernmentSecretComplex/SecretComplex_Computer_Oneshot_15}}
 {
  id=WavePlayer_2
  properties {bLooping=true, soundWaveAssetPtr=/Game/Sounds/Apollo/Biomes/GovernmentSecretComplex/SecretComplex_WallTab_Loop_02}
 }
 {id=WavePlayer_3, properties {soundWaveAssetPtr=/Game/Sounds/Apollo/Biomes/GovernmentSecretComplex/SecretComplex_Computer_Oneshot_12}}
 {
  id=WavePlayer_4
  properties {bLooping=true, soundWaveAssetPtr=&/Government_Computer_Idle_Activity_Buzz_Loop_02}
 }
]
properties {
 attenuationSettings=/Game/Sounds/Artemis/Environment_Actors/Computer_Screen/Government_Computer_Smaller_Att
 bOverrideConcurrency=true
 maxDistance=700
 pitchMultiplier=1.3
 soundClassObject=&Fort_Gadgets
 volumeMultiplier=0.25
 attenuationOverrides {
  bEnableFocusInterpolation=true
  falloffDistance=600
  nonFocusVolumeAttenuation=0.3
  occlusionVolumeAttenuation=0.3
  stereoSpread=0
  attenuationShapeExtents {x=200}
 }
 concurrencyOverrides {maxCount=1}
}
